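; I_ValetLightSpawnerInterface is a Blueprint interface for valet light spawners.
; It declares a single function, SetEmissiveLightsEnabled, which takes a Boolean to switch emissive lights on or off.
; The file holds only the signature, so implementing Blueprints define the actual lighting behavior.
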(fn SetEmissiveLightsEnabled
  (params (enabled :type Boolean)))
