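; Vehicle component that spawns cosmetic spotlights at configured transforms and skeletal mesh sockets, capped by a maximum count.
; The Valet.EnablePhysicalLights CVar creates or clears them and sets whether they need a driver and evening or night.
; Driver, disabled-state and day-phase changes queue one randomly delayed timer check, staggering updates per instance.
; That check toggles visibility and emissives through Valet_BasicCar or an interface, then notifies listeners; disabled vehicles stay dark.
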(event EventBeginPlay
  (Networking|CanExecuteCosmeticEvents
    (:True
      (Utilities|FlowControl|Sequence
        (:then_0
          (bind _output_get (Variables|Default|SetmyRandomInterval (Math|Random|RandomFloatinRange 0.1 0.5)))
          (Utilities|FlowControl|Delay _output_get)
          (DataDrivenCVar|BindEventtoOnDataDrivenCVarDelegate (EngineSubsystems|GetDataDrivenCVarEngineSubsystem) (EventDispatchers|CreateEvent :!event handleCvarChanged))
          (CallFunction|HandleCvarChanged "Valet.EnablePhysicalLights"))
        (:then_1
          (bind _asfort_athena_vehicle (Utilities|Casting|CastToFortAthenaVehicle (Components|GetOwner)))
          (Vehicle|BindEventtoOnPawnEnteredVehicleatSeat _asfort_athena_vehicle (EventDispatchers|CreateEvent :!event Handle%20Pawn%20Enter%20Seat))
          (Vehicle|BindEventtoOnPawnExitedVehiclefromSeat _asfort_athena_vehicle (EventDispatchers|CreateEvent :!event handlePawnExitSeat))
          (EventDispatchers|BindEventtoOnDisabledChangedDelegate _asfort_athena_vehicle (EventDispatchers|CreateEvent :!event handleDisabledChanged)))))))

(fn addLight
  (params (RelativeTransform :type Transform) (socketName :type Name))
  (if (< (Utilities|Array|Length (Variables|Default|GetcurrentLightComponents)) (Variables|Default|GetmaxLightCount))
    (Utilities|FlowControl|Sequence
      (:then_0
        (bind _returnvalue (Game|AddComponentbyClass (Components|GetOwner) (Variables|ValetLights|GetspotLightClass) false (Variables|GetRelativeTransform)))
        (bind _output_get (LocalVariables|Setl_newComp _returnvalue))
        (Utilities|Array|Add (Variables|Default|GetcurrentLightComponents) _output_get)
        (Rendering|SetVisibility (LocalVariables|Getl_newComp) false false) ; Not sure if it should be enabled yet, wait for validation
      )
      (:then_1
        (if (Utilities|Name|NotEqual(Name) (Variables|GetsocketName) "None")
          (bind _skeletalmesh (CallFunction|GetOwnerSkeletalMeshComp))
          (Transformation|AttachComponentToComponent (LocalVariables|Getl_newComp) _skeletalmesh (Variables|GetsocketName) "SnapToTarget" "SnapToTarget" "SnapToTarget" true)))
      (:then_2
        (return (LocalVariables|Getl_newComp))))))

(fn clearLights
  (for _array_element (Variables|Default|GetcurrentLightComponents)
    (Components|DestroyComponent _array_element))
  (Utilities|Array|Clear (Variables|Default|GetcurrentLightComponents)))

(fn GetOwnerSkeletalMeshComp
  (bind _asfort_athena_vehicle (Utilities|Casting|CastToFortAthenaVehicle (Components|GetOwner)))
  (return (Variables|/Script/FortniteGame.FortAthenaSKVehicle|GetSkeletalMesh _asfort_athena_vehicle)))

(fn HandlePawnEnterSeat
  (params (Vehicle :type Fort%20Vehicle%20Interface%20Interface%20%28by%20ref%29) (PlayerPawn :type Fort%20Pawn%20Object%20Reference) (SeatIndex :type Integer))
  (if (== SeatIndex 0)
    (if (not (Variables|Default|GetownerIsBeingDriven))
      (Variables|Default|SetownerIsBeingDriven true)
      (CallFunction|ValidateNextFrame)
      (TimeOfDay|RegisterforDayPhaseChangedEvent (TimeOfDay|GetContextualTimeOfDayManager) (EventDispatchers|CreateEvent :!event handleTimeOfDayChanged) false))))

(fn handleCvarChanged
  (params (cvarName :type String))
  (if (Utilities|String|EqualExactly(String) cvarName "Valet.EnablePhysicalLights")
    (CallFunction|ValidateCvarEnablePhysicalLights)))

(fn handleDisabledChanged
  (params (disabled :type Boolean))
  (CallFunction|ValidateNextFrame))

(fn handlePawnExitSeat
  (params (Vehicle :type Fort%20Vehicle%20Interface%20Interface%20%28by%20ref%29) (PlayerPawn :type Fort%20Pawn%20Object%20Reference) (SeatIndex :type Integer))
  (bind _returnvalue (TimeOfDay|GetContextualTimeOfDayManager))
  (if (== SeatIndex 0)
    (if (Variables|Default|GetownerIsBeingDriven)
      (Variables|Default|SetownerIsBeingDriven false)
      (CallFunction|ValidateNextFrame)
      (Utilities|IsValid (Utilities|ToObject(Interface) _returnvalue)
        (:Is%20Valid
          (TimeOfDay|UnregisterforDayPhaseChangedEvent _returnvalue self))) ; Prevents an error on PIE end
    )))

(fn handleTimeOfDayChanged
  (params (CurrentDayPhase :type EFortDayPhase%20Enum :default "Morning") (PreviousDayPhase :type EFortDayPhase%20Enum :default "Morning") (bAtCreation :type Boolean))
  (CallFunction|ValidateNextFrame))

(fn hasAnyLights
  (return (> (Utilities|Array|Length (Variables|Default|GetcurrentLightComponents)) 0)))

(fn isOwnerBeingDriven
  (return (Variables|Default|GetownerIsBeingDriven)))

(fn SetValetLightColor
  (params (NewLightColor :type Linear%20Color%20Structure))
  (Variables|ValetLights|SetoverrideLightColor NewLightColor)
  (for _array_element (Variables|Default|GetcurrentLightComponents)
    (Rendering|Components|Light|SetLightColor _array_element (Variables|ValetLights|GetoverrideLightColor) true)))

(fn shouldLightsBeOn
  (bind _asfort_vehicle_interface (Utilities|Casting|CastToFortVehicleInterface (Components|GetOwner))
    (:then
      (if (Vehicle|IsDisabled _asfort_vehicle_interface)
        (return false)
        (else
          _)))
    (:CastFailed
      _))
  (if (> (Variables|Default|GetcurrentCVarValue_EnabledState) 1)
    (if (> (Variables|Default|GetcurrentCVarValue_EnabledState) 2)
      (return true)
      (else
        (return (CallFunction|IsOwnerBeingDriven))))
    (else
      (return (and (or (Utilities|Enum|Equal(Enum) (TimeOfDay|GetCurrentDayPhase :!class /Script/FortniteGame.FortKismetLibrary) "Evening") (Utilities|Enum|Equal(Enum) (TimeOfDay|GetCurrentDayPhase :!class /Script/FortniteGame.FortKismetLibrary) "Night")) (CallFunction|IsOwnerBeingDriven))))))

(fn spawnLights
  (bind _l_lights (LocalVariables|Getl_lights))
  (if (not (ValetLights|HasAnyLights))
    (Utilities|FlowControl|Sequence
      (:then_0
        ;[1 1 1] Populate lights using transforms
        (for _array_element (Variables|ValetLights|GetadditionalLightTransforms)
          (if (not (Math|Transform|Equal(Transform) _array_element (Math|Transform|MakeTransform "0, 0, 0")))
            (CallFunction|AddLight _array_element))))
      (:then_1
        ;[1 1 1] Populate lights using sockets
        (bind _skeletalmesh (CallFunction|GetOwnerSkeletalMeshComp))
        (LocalVariables|Setl_skelComp _skeletalmesh)
        (bind _result (Utilities|Set|ToArray (Variables|ValetLights|Getlight_SocketNames)))
        (for _array_element_2 _result
          (if (Transformation|DoesSocketExist (LocalVariables|Getl_skelComp) _array_element_2)
            (bind _newlight (CallFunction|AddLight (Transformation|GetSocketTransform (LocalVariables|Getl_skelComp) _array_element_2 "RTS_Actor") _array_element_2))
            (Utilities|IsValid _newlight
              (:Is%20Valid
                (Utilities|Array|Add (LocalVariables|Getl_lights) _newlight))))))
      (:then_2
        (Default|CallOnNewLightsSpawned _l_lights)
        (return _l_lights)))))

(fn validateCarLights
  (bind _returnvalue (Components|GetOwner))
  (Utilities|Time|ClearandInvalidateTimerbyHandle (Variables|Default|GettimerHandle_validateCarLights))
  (LocalVariables|Setl_on (CallFunction|ShouldLightsbeOn))
  (if (Math|Boolean|NotEqual(Boolean) (LocalVariables|Getl_on) (Variables|Default|GetareLightsOn))
    (Variables|Default|SetareLightsOn (LocalVariables|Getl_on))
    (Utilities|FlowControl|Sequence
      (:then_0
        (for _array_element (Variables|Default|GetcurrentLightComponents)
          (Rendering|SetVisibility _array_element (Variables|Default|GetareLightsOn) false)))
      (:then_1
        (bind _asvalet_basic_car_vehicle (Utilities|Casting|CastToValet_BasicCar_Vehicle _returnvalue)
          (:then
            (Class|ValetBasicCarVehicle|SetHeadlightEmissiveOn _asvalet_basic_car_vehicle (Variables|Default|GetareLightsOn)))
          (:CastFailed
            (if (Utilities|DoesObjectImplementInterface _returnvalue "/Valet/BP/Lights/I_ValetLightSpawnerInterface.I_ValetLightSpawnerInterface_C")
              (Class|IValetLightSpawnerInterface|SetEmissiveLightsEnabled(Message) _returnvalue (Variables|Default|GetareLightsOn))))))
      (:then_2
        (Default|CallOnLightEnabledChanged (Variables|Default|GetareLightsOn))))))

(fn validateCvar_EnablePhysicalLights
  (bind _returnvalue (Development|GetConsoleVariableIntValue "Valet.EnablePhysicalLights"))
  (Variables|Default|SetcurrentCVarValue_EnabledState _returnvalue)
  (if (> _returnvalue 0)
    (CallFunction|SpawnLights)
    (CallFunction|ValidateNextFrame)
    (else
      (CallFunction|ClearLights))))

(fn validateNextFrame
  (if (not (Utilities|Time|IsValidTimerHandle (Variables|Default|GettimerHandle_validateCarLights)))
    (bind _returnvalue (Utilities|Time|SetTimerbyEvent (EventDispatchers|CreateEvent :!event validateCarLights) (Variables|Default|GetmyRandomInterval)))
    (Variables|Default|SettimerHandle_validateCarLights _returnvalue)))
